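# Control Rig for the male large WarmShadeWeasel hood, posing its bones procedurally from body motion, branched on HoodUp.
# Hood down pins the sides to the clavicles, then layers correctives weighted by spherical pose readers, spine angles and head-to-arm distance.
# A falling offset, suppressed by Parachute, is smoothed with Kalman filters chosen by SkyDiving.
# Hood up uses a head pose reader and remapped head rotations; some constraint and transform nodes are left unconnected.
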ControlRigBlueprint /BRCosmetics/Characters/Player/Male/Large/Bodies/M_LRG_WarmShadeWeasel_HoodUp/Meshes/Parts/M_LRG_WarmShadeWeasel_FaceAccHood_CtrlRig
graphs [
 RigVMFunctionLibrary {links [], nodes []}
 RigVMModel {nodes [
   RigUnit_GetRelativeTransformForItem = &RigUnit_GetRelativeTransformForItem(Child=(Type=Bone,Name="head"), Parent=(Type=Null,Name="head_null"), bChildInitial=False, bParentInitial=False)
   RigVMFunction_MathQuaternionToEuler = &RigVMFunction_MathQuaternionToEuler(RotationOrder=ZYX, Value=@RigUnit_GetRelativeTransformForItem.RelativeTransform.Rotation)
   n1 = &RigVMFunction_MathDoubleRemap(SourceMaximum=-35, SourceMinimum=0, TargetMaximum=1, TargetMinimum=0, Value=@RigVMFunction_MathQuaternionToEuler.Result.Z)
   Subtract = &RigVMFunction_MathDoubleSub(A=1, B=@n1.Result)
   &RigUnit_ParentConstraint(
    AdvancedSettings=(InterpolationType=Average,RotationOrderForFilter=XZY)
    Child=(Type=Bone,Name="mstr_HoodUp_back_c")
    ExecutePin=()
    Parents=((Item=(Type=Bone,Name="spine_05"),Weight=1),(Item=(Type=Bone,Name="head"),Weight=1))
    Weight=1
    Parents.0.Weight=@n1.Result
    Parents.1.Weight=@Subtract.Result
   )
   Remap = &RigVMFunction_MathDoubleRemap(SourceMaximum=-20, SourceMinimum=0, TargetMaximum=1, TargetMinimum=0, Value=@RigVMFunction_MathQuaternionToEuler.Result.Z)
   > &RigUnit_ModifyTransforms(
    ItemToModify=((Item=(Type=Bone,Name="mstr_HoodUp_back_c"),Transform=(Rotation=(X=0,Y=0,Z=0.1731,W=0.9849),Translation=(X=-8.744,Y=-7.8004,Z=-6.7341),Scale3D=(X=1,Y=1,Z=1))))
    WeightMinimum=0
    Weight=@Remap.Result
   )
   RigUnit_BeginExecution = &RigUnit_BeginExecution()
   RigVMFunction_ControlFlowBranch = &RigVMFunction_ControlFlowBranch(ExecuteContext=@RigUnit_BeginExecution.ExecutePin, Condition=RigVMVariableNode() {variable=HoodUp}.Value)
   n2 = &RigUnit_SphericalPoseReader(
    ActiveRegionScaleFactors=(PositiveWidth=1,NegativeWidth=2,PositiveHeight=1,NegativeHeight=2)
    ActiveRegionSize=0.05
    Debug=(bDrawDebug=false,bDraw2D=false,bDrawLocalAxes=False,DebugScale=25,DebugSegments=20,DebugThickness=0.25)
    DriverAxis=(X=1,Y=0,Z=0)
    DriverItem=(Type=Bone,Name="head")
    FalloffRegionScaleFactors=(PositiveWidth=1.2,NegativeWidth=1,PositiveHeight=1.2,NegativeHeight=1)
    FalloffSize=0.5
    FlipHeightScaling=False
    FlipWidthScaling=true
    OptionalParentItem=(Type=Bone,Name="spine_04")
    RotationOffset=(X=-47.2973,Y=-54.1955,Z=-19.739)
    ExecutePin=@RigVMFunction_ControlFlowBranch.False
   )
   > &RigUnit_ParentConstraint(
    AdvancedSettings=(InterpolationType=Average,RotationOrderForFilter=XZY)
    Child=(Type=Bone,Name="mstr_HoodDown_l")
    Parents=((Item=(Type=Bone,Name="clavicle_l"),Weight=1))
    Weight=1
   )
   > &RigUnit_ParentConstraint(
    AdvancedSettings=(InterpolationType=Average,RotationOrderForFilter=XZY)
    Child=(Type=Bone,Name="mstr_HoodDown_r")
    Parents=((Item=(Type=Bone,Name="clavicle_r"),Weight=1))
    Weight=1
   )
   > n3 = &RigUnit_SphericalPoseReader(
    ActiveRegionScaleFactors=(PositiveWidth=1,NegativeWidth=1,PositiveHeight=1,NegativeHeight=1)
    ActiveRegionSize=0.2
    Debug=(bDrawDebug=false,bDraw2D=false,bDrawLocalAxes=False,DebugScale=25,DebugSegments=20,DebugThickness=0.25)
    DriverAxis=(X=1,Y=0,Z=0)
    DriverItem=(Type=Bone,Name="head")
    FalloffRegionScaleFactors=(PositiveWidth=1,NegativeWidth=1,PositiveHeight=1,NegativeHeight=1)
    FalloffSize=0.6
    FlipHeightScaling=False
    FlipWidthScaling=false
    OptionalParentItem=(Type=Bone,Name="spine_04")
    RotationOffset=(X=-49.4384,Y=-33.6729,Z=-50.3061)
   )
   > &RigUnit_ModifyTransforms(
    ItemToModify="((Item=(Type=Bone,Name=\"mstr_HoodDown_back_c\"),Transform=(Rotation=(X=0,Y=0,Z=-0.1241,W=-0.9923),Translation=(X=16.901,Y=-6.1562,Z=-19.7211),Scale3D=(X=1,Y=1,Z=1.2804))),(Item=(Type=Bone,Name=\"mstr_HoodDown_l\"),Transform=(Rotation=(X=0,Y=-0.1135,Z=0,W=-0.9 ...(truncated)"
    WeightMinimum=0
    Weight=@n3.OutputParam
   )
   > &RigUnit_ModifyTransforms(
    ItemToModify="((Item=(Type=Bone,Name=\"mstr_HoodDown_back_c\"),Transform=(Rotation=(X=0,Y=0,Z=0,W=1),Translation=(X=0,Y=0,Z=0),Scale3D=(X=1,Y=1,Z=1))),(Item=(Type=Bone,Name=\"mstr_HoodDown_r\"),Transform=(Rotation=(X=0,Y=0,Z=0,W=-1),Translation=(X=6.2123,Y=-7.9794,Z=5.3564) ...(truncated)"
    WeightMinimum=0
    Weight=@n2.OutputParam
   )
   > n4 = &RigUnit_SphericalPoseReader(
    ActiveRegionScaleFactors=(PositiveWidth=1,NegativeWidth=1,PositiveHeight=1,NegativeHeight=1)
    ActiveRegionSize=0.05
    Debug=(bDrawDebug=false,bDraw2D=false,bDrawLocalAxes=False,DebugScale=25,DebugSegments=20,DebugThickness=0.25)
    DriverAxis=(X=1,Y=0,Z=0)
    DriverItem=(Type=Bone,Name="head")
    FalloffRegionScaleFactors=(PositiveWidth=1,NegativeWidth=1,PositiveHeight=1,NegativeHeight=1)
    FalloffSize=0.4
    FlipHeightScaling=False
    FlipWidthScaling=False
    OptionalParentItem=(Type=Bone,Name="spine_04")
    RotationOffset=(X=-94.594,Y=-27.4718,Z=31.071)
   )
   > &RigUnit_ModifyTransforms(
    ItemToModify="((Item=(Type=Bone,Name=\"mstr_HoodDown_l\"),Transform=(Rotation=(X=0,Y=0,Z=0,W=1),Translation=(X=0,Y=0,Z=0),Scale3D=(X=1,Y=1,Z=1))),(Item=(Type=Bone,Name=\"mstr_HoodDown_back_c\"),Transform=(Rotation=(X=0,Y=0,Z=0,W=-1),Translation=(X=0,Y=0,Z=8.4968),Scale3D=(X ...(truncated)"
    WeightMinimum=0
    Weight=@n4.OutputParam
   )
   > n5 = &RigUnit_SphericalPoseReader(
    ActiveRegionScaleFactors=(PositiveWidth=1,NegativeWidth=1,PositiveHeight=1,NegativeHeight=1)
    ActiveRegionSize=0.27
    Debug=(bDrawDebug=false,bDraw2D=false,bDrawLocalAxes=False,DebugScale=25,DebugSegments=20,DebugThickness=0.25)
    DriverAxis=(X=1,Y=0,Z=0)
    DriverItem=(Type=Bone,Name="clavicle_l")
    FalloffRegionScaleFactors=(PositiveWidth=1,NegativeWidth=1,PositiveHeight=1,NegativeHeight=1)
    FalloffSize=0.45
    FlipHeightScaling=False
    FlipWidthScaling=False
    OptionalParentItem=(Type=Bone,Name="None")
    RotationOffset=(X=0,Y=0,Z=0)
   )
   > &RigUnit_ModifyTransforms(
    ItemToModify=((Item=(Type=Bone,Name="mstr_HoodDown_l"),Transform=(Rotation=(X=0,Y=0,Z=0,W=1),Translation=(X=-7,Y=17,Z=-12),Scale3D=(X=1,Y=1,Z=1))))
    WeightMinimum=0
    Weight=@n5.OutputParam
   )
   > n6 = &RigUnit_SphericalPoseReader(
    ActiveRegionScaleFactors=(PositiveWidth=1,NegativeWidth=1,PositiveHeight=1,NegativeHeight=1)
    ActiveRegionSize=0.27
    Debug=(bDrawDebug=false,bDraw2D=false,bDrawLocalAxes=False,DebugScale=25,DebugSegments=20,DebugThickness=0.25)
    DriverAxis=(X=-1,Y=0,Z=0)
    DriverItem=(Type=Bone,Name="clavicle_r")
    FalloffRegionScaleFactors=(PositiveWidth=1,NegativeWidth=1,PositiveHeight=1,NegativeHeight=1)
    FalloffSize=0.45
    FlipHeightScaling=False
    FlipWidthScaling=False
    OptionalParentItem=(Type=Bone,Name="None")
    RotationOffset=(X=180,Y=0,Z=0)
   )
   > &RigUnit_ModifyTransforms(
    ItemToModify=((Item=(Type=Bone,Name="mstr_HoodDown_r"),Transform=(Rotation=(X=-0.186,Y=-0.208,Z=0.1241,W=-0.9522),Translation=(X=5.7966,Y=-11.7365,Z=-6.2511),Scale3D=(X=1,Y=1,Z=1))))
    WeightMinimum=0
    Weight=@n6.OutputParam
   )
   RigUnit_GetTransform = &RigUnit_GetTransform(Item=(Type=Bone,Name="spine_05"), Space=LocalSpace, bInitial=False)
   ToVectors = /Script/RigVM.RigVMFunction_MathTransformToVectors(Value=&RigUnit_GetRelativeTransformForItem(Child=(Type=Bone,Name="spine_05"), Parent=(Type=Bone,Name="clavicle_r"), bChildInitial=False, bParentInitial=False).RelativeTransform) {is=struct RigVMUnitNode, template="ToVectors::Execute(in Value,out Forward,out Right,out Up)"}
   RigVMFunction_MathVectorAngle = /Script/RigVM.RigVMFunction_MathVectorAngle(B=(X=-437,Y=80,Z=-895), A=@ToVectors.Forward) {is=struct RigVMUnitNode, template="AngleBetween::Execute(in A,in B,out Result)"}
   Multiply = &RigVMFunction_MathDoubleMul(
    B=&RigVMFunction_MathDoubleRemap(
     SourceMaximum=-21
     SourceMinimum=0
     TargetMaximum=1
     TargetMinimum=0
     Value=&RigVMFunction_MathQuaternionToEuler(RotationOrder=ZYX, Value=@RigUnit_GetTransform.Transform.Rotation).Result.Y
    ).Result
    A=&RigVMFunction_MathDoubleRemap(SourceMaximum=0.01, SourceMinimum=0.2, TargetMaximum=1, TargetMinimum=0, Value=@RigVMFunction_MathVectorAngle.Result).Result
   )
   > &RigUnit_ModifyTransforms(
    ItemToModify="((Item=(Type=Bone,Name=\"mstr_HoodDown_r\"),Transform=(Rotation=(X=0,Y=0,Z=0,W=-1),Translation=(X=-3.9417,Y=-2.0415,Z=5.4673),Scale3D=(X=1,Y=3.0161,Z=4.7478))),(Item=(Type=Bone,Name=\"mstr_HoodDown_back_c\"),Transform=(Rotation=(X=0,Y=0,Z=0,W=-1),Translation=( ...(truncated)"
    WeightMinimum=0
    Weight=@Multiply.Result
   )
   RigVMFunction_MathBoolToFloat = &RigVMFunction_MathBoolToFloat(Value=RigVMDispatchNode(True=False, False=RigVMVariableNode() {variable=Falling}.Value, Condition=RigVMVariableNode() {variable=Parachute}.Value) {template="DISPATCH_RigVMDispatch_If(in Condition,in True,in False,out Result)"}.Result)
   n7 = &RigVMFunction_KalmanFloat(BufferSize=16, Value=@RigVMFunction_MathBoolToFloat.Result) {template="KalmanFilter::Execute(in Value,in BufferSize,out Result)"}
   n8 = RigVMVariableNode() {variable=SkyDiving}
   > RigVMVariableNode(Value=RigVMDispatchNode(
     False=&RigVMFunction_KalmanFloat(BufferSize=16, Value=@RigVMFunction_MathBoolToFloat.Result) {template="KalmanFilter::Execute(in Value,in BufferSize,out Result)"}.Result
     True=@n7.Result
     Condition=@n8.Value
    ) {template="DISPATCH_RigVMDispatch_If(in Condition,in True,in False,out Result)"}.Result) {variable=StartfallingInterpolation}
   KalmanFilter = &RigVMFunction_KalmanFloat(BufferSize=4, Value=@RigVMFunction_MathBoolToFloat.Result) {template="KalmanFilter::Execute(in Value,in BufferSize,out Result)"}
   > RigVMVariableNode(Value=RigVMDispatchNode(False=@KalmanFilter.Result, True=@n7.Result, Condition=@n8.Value) {template="DISPATCH_RigVMDispatch_If(in Condition,in True,in False,out Result)"}.Result) {variable=EndfallingInterpolation}
   DISPATCH_RigVMDispatch_If = RigVMDispatchNode(
    Condition=RigVMVariableNode() {variable=Falling}.Value
    True=RigVMVariableNode() {variable=StartfallingInterpolation}.Value
    False=RigVMVariableNode() {variable=EndfallingInterpolation}.Value
   ) {template="DISPATCH_RigVMDispatch_If(in Condition,in True,in False,out Result)"}
   > &RigUnit_ModifyTransforms(
    ItemToModify="((Item=(Type=Bone,Name=\"mstr_HoodDown_wobble\"),Transform=(Rotation=(X=0,Y=0,Z=0.0872,W=0.9962),Translation=(X=5,Y=-1,Z=0),Scale3D=(X=1,Y=1,Z=1))),(Item=(Type=Bone,Name=\"mstr_HoodDown_back_c\"),Transform=(Rotation=(X=0,Y=0,Z=0,W=-1),Translation=(X=-1,Y=-1,Z= ...(truncated)"
    WeightMinimum=0
    Weight=@DISPATCH_RigVMDispatch_If.Result
   )
   RigVMFunction_MathVectorDistance = &RigVMFunction_MathVectorDistance(
    A=&RigUnit_GetTransform(Item=(Type=Bone,Name="head"), bInitial=False).Transform.Translation
    B=&RigUnit_GetTransform(Item=(Type=Bone,Name="upperarm_l"), bInitial=False).Transform.Translation
   )
   > &RigUnit_ModifyTransforms(
    ItemToModify=((Item=(Type=Bone,Name="mstr_HoodDown_l"),Transform=(Rotation=(X=0,Y=0,Z=0,W=1),Translation=(X=-6.3057,Y=4.6802,Z=5.9789),Scale3D=(X=1,Y=1,Z=1))))
    WeightMinimum=0
    Weight=&RigVMFunction_MathDoubleRemap(SourceMaximum=18, SourceMinimum=30, TargetMaximum=1, TargetMinimum=0, bClamp=False, Value=@RigVMFunction_MathVectorDistance.Result).Result
   )
   RigVMFunction_Sequence = &RigVMFunction_Sequence(ExecuteContext=@RigVMFunction_ControlFlowBranch.True)
   DISPATCH_RigVMDispatch_Print = RigVMDispatchNode(
    Enabled=false
    ScreenColor=(R=1,G=1,B=1,A=1)
    ScreenDuration=0.05
    Value=@RigVMFunction_MathQuaternionToEuler.Result
    ExecuteContext=@RigVMFunction_Sequence.A
   ) {template="DISPATCH_RigVMDispatch_Print(in Prefix,in Value,in Enabled,in ScreenDuration,in ScreenColor)"}
   > n9 = &RigUnit_SphericalPoseReader(
    ActiveRegionScaleFactors=(PositiveWidth=1,NegativeWidth=1,PositiveHeight=1,NegativeHeight=1)
    ActiveRegionSize=0.1
    Debug=(bDrawDebug=true,bDraw2D=False,bDrawLocalAxes=False,DebugScale=25,DebugSegments=20,DebugThickness=0.25)
    DriverAxis=(X=1,Y=0,Z=0)
    DriverItem=(Type=Bone,Name="head")
    FalloffRegionScaleFactors=(PositiveWidth=1,NegativeWidth=1,PositiveHeight=1,NegativeHeight=1)
    FalloffSize=0.4
    FlipHeightScaling=False
    FlipWidthScaling=False
    OptionalParentItem=(Type=Bone,Name="spine_04")
    RotationOffset=(X=-70,Y=0,Z=15)
   )
   > &RigUnit_ModifyTransforms(
    ItemToModify="((Item=(Type=Bone,Name=\"mstr_HoodUp_back_c\"),Transform=(Rotation=(X=0.0471,Y=-0.1897,Z=-0.0091,W=-0.9807),Translation=(X=-7.0124,Y=-3.3889,Z=0.6774),Scale3D=(X=1,Y=1,Z=1))),(Item=(Type=Bone,Name=\"mstr_HoodDown_nape_c\"),Transform=(Rotation=(X=0,Y=0,Z=0,W=1) ...(truncated)"
    WeightMinimum=0
    Weight=@n9.OutputParam
   )
   &RigUnit_ModifyTransforms(
    ItemToModify="((Item=(Type=Bone,Name=\"mstr_HoodUp_Front_r\"),Transform=(Rotation=(X=-0.0468,Y=0.2113,Z=0.2113,W=-0.9532),Translation=(X=-9.1641,Y=-1.1395,Z=-10.4292),Scale3D=(X=1,Y=1,Z=1))),(Item=(Type=Bone,Name=\"mstr_HoodUp_Front_l\"),Transform=(Rotation=(X=-0.342,Y=0,Z= ...(truncated)"
    WeightMinimum=0
    Weight=&RigVMFunction_MathDoubleRemap(SourceMaximum=-45, SourceMinimum=0, TargetMaximum=1, TargetMinimum=0, bClamp=False, Value=@RigVMFunction_MathQuaternionToEuler.Result.X).Result
    ExecutePin=@RigVMFunction_Sequence.B
   )
   > &RigUnit_ModifyTransforms(
    ItemToModify="((Item=(Type=Bone,Name=\"mstr_HoodUp_Front_r\"),Transform=(Rotation=(X=-0.1305,Y=0,Z=0,W=-0.9914),Translation=(X=-6.8698,Y=-2.0748,Z=3.4349),Scale3D=(X=1,Y=1,Z=1))),(Item=(Type=Bone,Name=\"mstr_HoodUp_Front_l\"),Transform=(Rotation=(X=0.0057,Y=-0.0432,Z=0.1304 ...(truncated)"
    WeightMinimum=0
    Weight=&RigVMFunction_MathDoubleRemap(SourceMaximum=45, SourceMinimum=0, TargetMaximum=1, TargetMinimum=0, bClamp=False, Value=@RigVMFunction_MathQuaternionToEuler.Result.X).Result
   )
   n10 = &RigVMFunction_MathDoubleRemap(
    SourceMaximum=69
    SourceMinimum=0
    TargetMaximum=1
    TargetMinimum=0
    bClamp=False
    Value=&RigVMFunction_MathQuaternionToEuler(
     RotationOrder=ZYX
     Value=&RigUnit_GetRelativeTransformForItem(Child=(Type=Bone,Name="head"), Parent=(Type=Bone,Name="spine_05"), bChildInitial=False, bParentInitial=False).RelativeTransform.Rotation
    ).Result.X
   )
   > RigVMDispatchNode(Enabled=false, ScreenColor=(R=1,G=1,B=1,A=1), ScreenDuration=0.05, Value=@n10.Result) {template="DISPATCH_RigVMDispatch_Print(in Prefix,in Value,in Enabled,in ScreenDuration,in ScreenColor)"}
   > &RigUnit_ModifyTransforms(
    ItemToModify="((Item=(Type=Bone,Name=\"mstr_HoodUp_Front_r\"),Transform=(Rotation=(X=-0.1441,Y=0.1054,Z=0.4782,W=-0.8599),Translation=(X=1.6002,Y=-1.8178,Z=4.1936),Scale3D=(X=0.8761,Y=1.7229,Z=0.794))),(Item=(Type=Bone,Name=\"clavicle_r\"),Transform=(Rotation=(X=0,Y=0,Z=0,W ...(truncated)"
    WeightMinimum=0
    Weight=@n10.Result
   )
   &RigUnit_ModifyTransforms(
    ExecutePin=()
    ItemToModify="((Item=(Type=Bone,Name=\"mstr_HoodUp_back_c\"),Transform=(Rotation=(X=0,Y=0,Z=0,W=1),Translation=(X=0,Y=0,Z=0),Scale3D=(X=1,Y=1,Z=1))),(Item=(Type=Bone,Name=\"mstr_HoodDown_nape_c\"),Transform=(Rotation=(X=0,Y=0,Z=0,W=1),Translation=(X=0,Y=0,Z=0),Scale3D=(X=1, ...(truncated)"
    Weight=1
    WeightMinimum=0
   )
   &RigUnit_ModifyTransforms(
    ExecutePin=()
    ItemToModify=((Item=(Type=Bone,Name="mstr_HoodDown_l"),Transform=(Rotation=(X=0,Y=0,Z=0,W=1),Translation=(X=9.7505,Y=7.4032,Z=-0.2585),Scale3D=(X=1,Y=1,Z=0.8584))))
    WeightMinimum=0
    Weight=@n5.OutputParam
   )
   &RigUnit_ModifyTransforms(
    ExecutePin=()
    ItemToModify="((Item=(Type=Bone,Name=\"mstr_HoodDown_back_c\"),Transform=(Rotation=(X=0.0089,Y=-0.04,Z=-0.2173,W=-0.9752),Translation=(X=17.0174,Y=-8.6642,Z=-15.6143),Scale3D=(X=0.519,Y=0.7615,Z=1.1198))),(Item=(Type=Bone,Name=\"mstr_HoodDown_r\"),Transform=(Rotation=(X=0.3 ...(truncated)"
    Weight=1
    WeightMinimum=0
   )
  ]}
]
hierarchy [
 root {loc=<0 0 0>, rot=<0 0 0 1>}
 pelvis {parent=root, loc=<0 -0.7321 92.4218>, rot=<0.0127 0.707 -0.0127 -0.707>}
 spine_01 {parent=pelvis, loc=<5.6002 -0.5128 0>, rot=<0 0 0.034 -0.9994>}
 spine_02 {parent=spine_01, loc=<8.0753 0.0139 0>, rot=<0 0 0.0238 -0.9997>}
 spine_03 {parent=spine_02, loc=<7.0393 0 0>, rot=<0 0 0.0149 -0.9999>}
 spine_04 {parent=spine_03, loc=<8.1024 0 0>, rot=<0 0 -0.0132 -0.9999>}
 spine_05 {parent=spine_04, loc=<10.0146 0 0>, rot=<0 0 -0.0015 -1>}
 neck_01 {parent=spine_05, loc=<10.815 1.9196 0>, rot=<0 0 -0.1864 -0.9825>}
 neck_02 {parent=neck_01, loc=<4.7347 0.3045 0>, rot=<0 0 0.0363 -0.9993>}
 head {parent=neck_02, loc=<5.7632 0.5552 0>, rot=<0 0 0.115 -0.9934>}
 faceAccRoot {parent=head, loc=<0 0 0>, rot=<0 0 0 1>}
 mstr_HoodUp_Front_r {parent=faceAccRoot, loc=<-2.768 3.9429 11.8083>, rot=<-0.0197 -0.7241 -0.6889 0.0267>}
 dyn_HoodUp_Front_r {parent=mstr_HoodUp_Front_r, loc=<0 0 0>, rot=<0 0 0 1>}
 mstr_HoodUp_Front_l {parent=faceAccRoot, loc=<-2.7679 3.9429 -11.8083>, rot=<-0.7241 0.0197 -0.0267 -0.6889>}
 dyn_HoodUp_Front_l {parent=mstr_HoodUp_Front_l, loc=<0 0 0>, rot=<0 0 0 1>}
 mstr_faceAcc_head {parent=faceAccRoot, loc=<0 0 0>, rot=<0 0 0 1>}
 FACIAL_C_FacialRoot {parent=head, loc=<2.1743 -3.3889 -0.0002>, rot=<0.4879 -0.5118 -0.4879 -0.5118>}
 FACIAL_L_CheekInner {parent=FACIAL_C_FacialRoot, loc=<1.9467 -1.1328 9.1888>, rot=<-0.121 -0.0883 -0.1969 -0.9689>}
 FACIAL_R_CheekInner {parent=FACIAL_C_FacialRoot, loc=<-1.9371 -1.1306 9.1864>, rot=<-0.121 0.0886 0.1969 -0.9689>}
 FACIAL_L_CheekOuter {parent=FACIAL_C_FacialRoot, loc=<3.7448 -0.8667 7.7388>, rot=<0.0573 -0.3766 0.1012 -0.919>}
 FACIAL_R_CheekOuter {parent=FACIAL_C_FacialRoot, loc=<-3.7422 -0.8649 7.7359>, rot=<0.0573 0.3768 -0.1012 -0.919>}
 FACIAL_L_NasolabialBulge {parent=FACIAL_C_FacialRoot, loc=<3.6059 2.795 7.5698>, rot=<0.06 -0.3772 0.0247 -0.9239>}
 FACIAL_R_NasolabialBulge {parent=FACIAL_C_FacialRoot, loc=<-3.593 2.7823 7.5436>, rot=<0.06 0.3774 -0.0246 -0.9238>}
 FACIAL_L_NasolabialFurrow {parent=FACIAL_C_FacialRoot, loc=<2.8511 1.203 10.1415>, rot=<-0.0311 -0.1265 -0.3633 -0.9225>}
 FACIAL_R_NasolabialFurrow {parent=FACIAL_C_FacialRoot, loc=<-2.8329 1.2067 10.1365>, rot=<-0.0312 0.1267 0.3634 -0.9225>}
 FACIAL_L_CheekLower {parent=FACIAL_C_FacialRoot, loc=<3.9462 2.5381 6.783>, rot=<0.0366 -0.4266 0.0587 -0.9018>}
 FACIAL_R_CheekLower {parent=FACIAL_C_FacialRoot, loc=<-3.9216 2.5237 6.7531>, rot=<0.0367 0.4268 -0.0587 -0.9017>}
 FACIAL_C_Nose {parent=FACIAL_C_FacialRoot, loc=<0.0016 -1.6257 10.831>, rot=<-0.1281 0.0001 0 -0.9918>}
 FACIAL_C_NoseLower {parent=FACIAL_C_Nose, loc=<-0.0015 2.5372 -0.0469>, rot=<0.1305 0 0 -0.9914>}
 FACIAL_L_Nostril {parent=FACIAL_C_Nose, loc=<1.4777 1.949 -0.2497>, rot=<0.0173 0.1305 0.0023 -0.9913>}
 FACIAL_R_Nostril {parent=FACIAL_C_Nose, loc=<-1.4972 1.9631 -0.2657>, rot=<0.0173 -0.1305 -0.0023 -0.9913>}
 FACIAL_C_MouthUpper {parent=FACIAL_C_FacialRoot, loc=<0.0012 3.0528 8.6538>, rot=<0.0023 0.0001 0 -1>}
 FACIAL_C_LipUpper {parent=FACIAL_C_MouthUpper, loc=<-0.0064 -0.8354 2.7306>, rot=<-0.0872 0 0 -0.9962>}
 FACIAL_L_LipUpper {parent=FACIAL_C_MouthUpper, loc=<1.4885 -0.6048 2.3748>, rot=<-0.0778 -0.1601 -0.0379 -0.9833>}
 FACIAL_R_LipUpper {parent=FACIAL_C_MouthUpper, loc=<-1.5389 -0.6155 2.3747>, rot=<-0.0778 0.1601 0.0379 -0.9833>}
 FACIAL_L_LipUpperOuter {parent=FACIAL_C_MouthUpper, loc=<2.2363 -0.1996 1.6568>, rot=<-0.0984 -0.2293 -0.0983 -0.9634>}
 FACIAL_R_LipUpperOuter {parent=FACIAL_C_MouthUpper, loc=<-2.3072 -0.2224 1.6291>, rot=<-0.0984 0.2293 0.0983 -0.9634>}
 FACIAL_L_LipCorner {parent=FACIAL_C_MouthUpper, loc=<2.4795 0.2213 0.8899>, rot=<0 -0.2078 0 -0.9782>}
 FACIAL_R_LipCorner {parent=FACIAL_C_MouthUpper, loc=<-2.5414 0.1972 0.8511>, rot=<0 0.2078 0 -0.9782>}
 FACIAL_L_Masseter {parent=FACIAL_C_FacialRoot, loc=<6.3133 -0.2766 5.0208>, rot=<0.0716 0.113 0.0076 -0.991>}
 FACIAL_R_Masseter {parent=FACIAL_C_FacialRoot, loc=<-6.3072 -0.2827 5.0054>, rot=<0.0716 -0.1128 -0.0076 -0.991>}
 FACIAL_C_UnderChin {parent=FACIAL_C_FacialRoot, loc=<0.0102 8.2736 5.7884>, rot=<0.5018 0.0001 0.0001 -0.865>}
 FACIAL_L_UnderChin {parent=FACIAL_C_FacialRoot, loc=<3.7526 6.6589 3.9389>, rot=<0.4044 -0.3546 0.0832 -0.8389>}
 FACIAL_R_UnderChin {parent=FACIAL_C_FacialRoot, loc=<-3.7309 6.6415 3.9307>, rot=<0.4045 0.3548 -0.0831 -0.8388>}
 FACIAL_C_LowerLipRotation {parent=FACIAL_C_FacialRoot, loc=<0.0273 0.4183 3.4679>, rot=<0.3442 0.0001 0.0001 -0.9389>}
 FACIAL_C_MouthLower {parent=FACIAL_C_LowerLipRotation, loc=<-0.025 -1.3415 5.66>, rot=<-0.342 0 0 -0.9397>}
 FACIAL_C_LipLower {parent=FACIAL_C_MouthLower, loc=<0.0055 1.133 2.4349>, rot=<0.0872 0 0 -0.9962>}
 FACIAL_L_LipLower {parent=FACIAL_C_MouthLower, loc=<1.3642 1.1041 2.2072>, rot=<0.0927 -0.1294 0.0272 -0.9869>}
 FACIAL_R_LipLower {parent=FACIAL_C_MouthLower, loc=<-1.3745 1.056 2.1891>, rot=<0.0927 0.1294 -0.0272 -0.9869>}
 FACIAL_L_LipLowerOuter {parent=FACIAL_C_MouthLower, loc=<2.2291 0.6329 1.4725>, rot=<0.0913 -0.2321 0.0318 -0.9679>}
 FACIAL_R_LipLowerOuter {parent=FACIAL_C_MouthLower, loc=<-2.2563 0.6228 1.4435>, rot=<0.0913 0.2321 -0.0318 -0.9679>}
 FACIAL_C_Jaw {parent=FACIAL_C_FacialRoot, loc=<0.0273 0.4183 3.4679>, rot=<0.3442 0.0001 0.0001 -0.9389>}
 FACIAL_C_Jawline {parent=FACIAL_C_Jaw, loc=<-0.0218 3.1798 8.2939>, rot=<0.2586 0 0 -0.966>}
 FACIAL_L_ChinSide {parent=FACIAL_C_Jaw, loc=<3.4231 0.8203 7.7563>, rot=<-0.2208 -0.3738 0.0499 -0.8995>}
 FACIAL_R_ChinSide {parent=FACIAL_C_Jaw, loc=<-3.4811 0.83 7.7482>, rot=<-0.2208 0.3738 -0.0499 -0.8995>}
 FACIAL_C_Chin {parent=FACIAL_C_Jaw, loc=<-0.0233 1.3228 8.3122>, rot=<-0.2591 0 0 -0.9659>}
 FACIAL_L_Ear {parent=FACIAL_C_FacialRoot, loc=<6.5389 -1.7666 2.6306>, rot=<-0.024 -0.0006 -0.0261 -0.9994>}
 FACIAL_R_Ear {parent=FACIAL_C_FacialRoot, loc=<-6.5123 -1.7725 2.6576>, rot=<-0.024 0.0009 0.0262 -0.9994>}
 FACIAL_L_Jawline {parent=FACIAL_C_FacialRoot, loc=<4.8558 4.4269 5.2379>, rot=<0.2115 0.1577 -0.055 -0.963>}
 FACIAL_R_Jawline {parent=FACIAL_C_FacialRoot, loc=<-4.8264 4.4172 5.2234>, rot=<0.2115 -0.1575 0.0551 -0.963>}
 mstr_HoodUp_back_c {parent=head, loc=<2.9518 -12.948 0>, rot=<0.1036 -0.9946 0 0>}
 mstr_HoodDown_nape_c {parent=head, loc=<12.9459 -10.3724 0>, rot=<0.1036 -0.9946 0 0>}
 head_null {parent=neck_02, type=Null, loc=<5.7632 0.5552 0>, rot=<0 0 0.115 -0.9934>}
 mstr_HoodUp_back_r {parent=neck_01, loc=<5.8686 -13.1967 6.1725>, rot=<-0.0553 -0.9812 0.1766 0.0543>}
 mstr_HoodUp_back_l {parent=neck_01, loc=<5.8689 -13.1968 -6.1725>, rot=<0.9812 -0.0553 0.0543 -0.1766>}
 clavicle_l {parent=spine_05, loc=<2.5856 -0.2866 -4.3777>, rot=<0.0465 -0.7289 -0.009 -0.6829>}
 upperarm_l {parent=clavicle_l, loc=<18.2264 0.1342 -0.001>, rot=<0.0253 -0.2693 0.0013 -0.9627>}
 upperarm_twist_01_l {parent=upperarm_l, loc=<7.9796 0.0843 -0.6195>, rot=<0 0 0 -1>}
 deform_upperarm_bk_base_l {parent=upperarm_twist_01_l, loc=<-6.8983 -13.3826 -1.3258>, rot=<0 0 0 1>}
 deform_upperarm_bk_l {parent=deform_upperarm_bk_base_l, loc=<0 0 0>, rot=<0 0 0 1>}
 deform_upperarm_fr_base_l {parent=upperarm_twist_01_l, loc=<-8.3487 10.7117 -2.1455>, rot=<0 0 0 1>}
 deform_upperarm_fr_l {parent=deform_upperarm_fr_base_l, loc=<0 0 0>, rot=<0 0 0 1>}
 deform_upperarm_up_base_l {parent=upperarm_twist_01_l, loc=<-11.2012 -1.2234 8.7784>, rot=<0 0 0 1>}
 deform_upperarm_up_l {parent=deform_upperarm_up_base_l, loc=<0 0 0>, rot=<0 0 0 1>}
 clavicle_r {parent=spine_05, loc=<2.5855 -0.2866 4.3777>, rot=<0.7289 0.0465 -0.6829 0.009>}
 upperarm_r {parent=clavicle_r, loc=<-18.2264 -0.1342 0.001>, rot=<0.0253 -0.2693 0.0013 -0.9627>}
 upperarm_twist_01_r {parent=upperarm_r, loc=<-7.9796 -0.0843 0.6195>, rot=<0 0 0 -1>}
 deform_upperarm_fr_base_r {parent=upperarm_twist_01_r, loc=<8.3485 -10.7117 2.1457>, rot=<0 0 0 1>}
 deform_upperarm_fr_r {parent=deform_upperarm_fr_base_r, loc=<0 0 0>, rot=<0 0 0 1>}
 deform_upperarm_up_base_r {parent=upperarm_twist_01_r, loc=<11.2014 1.2235 -8.7787>, rot=<0 0 0 1>}
 deform_upperarm_up_r {parent=deform_upperarm_up_base_r, loc=<0 0 0>, rot=<0 0 0 1>}
 mstr_HoodDown_r {parent=spine_05, loc=<20.0622 2.7957 10.8492>, rot=<0.2257 -0.7615 0.5593 -0.2373>}
 mstr_HoodDown_l {parent=spine_05, loc=<20.0625 2.7957 -10.8492>, rot=<0.7615 0.2257 -0.2373 -0.5593>}
 mstr_HoodDown_back_c {parent=spine_05, loc=<25.9462 -5.6439 0>, rot=<0.1393 -0.9903 0 0>}
 mstr_HoodDown_wobble {parent=spine_05, loc=<15.318 -15.9847 0>, rot=<0 0 0 1>}
 dyn_HoodDown_wobble {parent=mstr_HoodDown_wobble, loc=<0 0 0>, rot=<0 0 0 1>}
]
variables [
 EndfallingInterpolation {type=double, default=0}
 Falling {type=bool, default=False}
 HoodUp {type=bool, default=False}
 Parachute {type=bool, default=False}
 SkyDiving {type=bool, default=False}
 StartfallingInterpolation {type=double, default=0}
]
