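# Physics asset for the hood-down face accessory part of the large male WarmShadeWeasel cosmetic character, apparently for secondary hood wobble.
# A capsule on dyn_HoodDown_wobble is constrained to a small sphere on mstr_HoodDown_wobble, with twist locked and narrow swing angles set.
# An unconstrained spine_05 body combines a capsule and a rotated box, likely as upper-torso collision for the hood.
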PhysicsAsset /BRCosmetics/Characters/Player/Male/Large/Bodies/M_LRG_WarmShadeWeasel_HoodUp/Meshes/Parts/M_LRG_WarmShadeWeasel_HoodDown_FaceAcc_Physics
bodies {
 dyn_HoodDown_wobble [{shapeType=Capsule, length=9.7236, radius=4.904, shapeName=dyn_HoodDown_wobble_capsule, rotation=<rot 0 23.4309 0>}]
 mstr_HoodDown_wobble [{shapeType=Sphere, radius=2.6396, shapeName=mstr_HoodDown_wobble_sphere}]
 spine_05 [
  {shapeType=Capsule, length=10.8004, radius=7.403, shapeName=spine_05_capsule, center=<3.0288 -19.6099 0>}
  {
   shapeType=Box
   extentX=26.0337
   extentY=21.9384
   extentZ=31.633
   shapeName=spine_05_box
   center=<6.603 -2.1816 0>
   rotation=<rot 0 28.3429 0>
  }
 ]
}
constraints [{
  bone1Name=dyn_HoodDown_wobble
  bone2Name=mstr_HoodDown_wobble
  swing1LimitDegrees=5
  swing2LimitDegrees=15
  twistLimitDegrees=45
  twistMotion=Locked
 }]
